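# Debug material in the Agent Jones butterfly quest VFX folder, flagged for Niagara mesh particles, holding scratch math for particle-space positioning.
# It takes world position relative to particle position and converts it through world, local and particle spaces, scaling one result down.
# A separate branch rotates world position over time around the particle's Z axis, pivoting on particle position.
# No node chain feeds a material output.
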Material /BattlepassS16/QuestActors/AgentJones/Butterfly/VFX/Debug
graph {
 properties {bUsedWithNiagaraMeshParticles=true, pixelDepthOffsetMode=PDOM_Legacy, usedShadingModels=Default Lit}
 nodes [
  Constant3Vector() {constant=<rgba 0 0 30 0>}
  n1 = Subtract(A=WorldPosition() {worldPositionShaderOffset=WPT_ExcludeAllShaderOffsets}.XYZ, B=ParticlePositionWS())
  Divide(A=Transform(None=@n1) {transformSourceType=TRANSFORMSOURCE_World, transformType=TRANSFORM_ParticleWorld}) {constB=100}
  n2 = Constant3Vector() {constant=<rgba 0 0 1 0>}
  RotateAboutAxis(
   NormalizedRotationAxis=Normalize(VectorInput=Transform(None=@n2) {transformSourceType=TRANSFORMSOURCE_ParticleWorld})
   RotationAngle=Time()
   PivotPoint=ParticlePositionWS()
   Position=WorldPosition().XYZ
  )
  Transform(None=Transform(None=@n1) {transformSourceType=TRANSFORMSOURCE_World, transformType=TRANSFORM_Local}) {transformSourceType=TRANSFORMSOURCE_Local, transformType=TRANSFORM_ParticleWorld}
 ]
}
