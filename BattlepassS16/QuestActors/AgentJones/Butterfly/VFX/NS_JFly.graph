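# NS_JFly is the Niagara VFX for the Agent Jones quest butterfly: a persistent mesh with Zero Point-style glow.
# Dark and pale-blue ribbons spawn by distance from User.SkeletalMesh bones, using bone-locked particles and velocity-integrated alpha.
# Layered flare sprites scale up with LOD distance, while a darkening backdrop fades far away and widens with User.Destabilization.
# GPU crystals tumble and twinkle, sine offsets bob render positions, and Audio_EMitter has no renderer.
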NiagaraSystem /BattlepassS16/QuestActors/AgentJones/Butterfly/VFX/NS_JFly
dependencies {
 usedDataInterfaces [&NiagaraDataInterfaceArrayInt32, &NiagaraDataInterfaceCurve]
 usedDynamicInputs [
  &MultiplyQuaternion
  &MakeQuatFromAxisAngle
  &Add_Float
  &self:ProjectSystemXAxisOnZPlane
 ]
 usedModules [
  &SystemState
  &self:SystemSpawnScript.SystemScriptSource.SystemScriptGraph.NiagaraNodeAssignment_0.SetVariables_#1
  &self:SystemSpawnScript.SystemScriptSource.SystemScriptGraph.NiagaraNodeAssignment_1.SetVariables_#2
 ]
}
emitters {
 Audio_EMitter {
  emitterUpdateScript [
   &EmitterState(Life Cycle Mode=::NewEnumerator1, Loop Duration=$System.LoopedAge, UseLoopDelay=false)
   &SpawnBurst_Instantaneous(Spawn Count=1, Use Spawn Probability=false)
  ]
  particleSpawnScript [&InitializeParticle(Color Mode=::NewEnumerator0, Lifetime=$System.NormalizedLoopAge, UsePositionOffset=false, Write Material Random=true)]
  particleUpdateScript [&ParticleState()]
  rendererClasses []
  renderers []
 }
 Crystals {
  simTarget=GPUComputeSim
  emitterUpdateScript [
   &EmitterState()
   /SetVariables(Emitter.Angle:NiagaraFloat=&Multiply_Float(
     A=1
     B=/Niagara/DynamicInputs/Angles/DegreesToRadians(AngleInDegrees:NiagaraFloat=&Sine(Normalized Angle=$Emitter.NormalizedLoopAge, Period=1, Scale=20, Bias=-0.9636))
    )) {isSetParameters=true}
   &SpawnRate(SpawnRate=4, Use Spawn Probability=false)
  ]
  particleSpawnScript [
   &InitializeParticle(
    Color Mode=::NewEnumerator0
    Lifetime=&RandomRangeFloat(Minimum=3, Maximum=6)
    Mesh Scale Mode=::NewEnumerator3
    Mesh Uniform Scale=&Multiply_Float(A=10, B=$User.Scale)
    Position Offset:NiagaraPosition=$Engine.Owner.Position
    Sprite Size Mode=::NewEnumerator3
    Uniform Sprite Size=&Multiply_Float(A=120, B=$User.Scale)
    UsePositionOffset=true
    Write Material Random=true
   )
   &self:Parabola_01(
    Output Max=1.1075
    Output Min=0.5592
    Rotation:Quat4f=$System.MeshOrientation
    Scale:NiagaraFloat=$User.Scale
    Skew Max=1.2091
    Skew Min=1.0566
    Width=189.3421
    YMax=293.7991
    Z Max=258.577
    Z Min=-423.0829
    falloff:NiagaraFloat=$User.Falloff
   ) {enabled=false, module=Parabola_01}
   &InitialMeshOrientation(Orientation Vector=&RandomVector(Vector Scale=1), Use Orientation Vector=true, Use Rotation Vector=false)
   &self:Butterfly_Flap_01(
    Flap Amplitude=0.1
    Rotation:Quat4f=$System.MeshOrientation
    Scale:NiagaraFloat=$User.Scale
    Time:NiagaraFloat=$System.NormalizedLoopAge
    Wave Amplitude=0.12
    Wave Length=3000
    Wave Offset=0.0091
   ) {enabled=false, module=Butterfly_Flap_01}
   &SphereLocation(
    MaskSpawnGroup=false
    Override Seed=false
    Sphere Origin:NiagaraPosition=$Engine.Owner.Position
    Sphere Radius=10
    Surface Only=false
    UseNonUniformScale=false
    UseOffset=false
    UseSphereOrientation=false
   )
  ]
  particleUpdateScript [
   &ParticleState()
   &CurlNoiseForce(Noise Strength=100, Pan Noise=false)
   &GravityForce(Gravity=<0 0 -50>)
   &PendulumSetup(
    Module.Enable Spring Driver=false
    Pendulum Length:NiagaraFloat=&FloatFromCurve(
     FloatCurve=DataInterface() {propertyValues {Curve {keys [
         <at 0.0045 0.0907 RCIM_Cubic>
         {arriveTangent=0.1632, interpMode=RCIM_Cubic, leaveTangent=0.1632, tangentMode=RCTM_User, time=0.5054, value=0.9681}
        ]}}}
     Scale Curve=&UniformAOrBFloat(
      A=20
      B=1000000000
      Randomness Mode=::NewEnumerator1
      Random Seed=HlslExpression() {hlslExpression=Particles.UniqueID}
      Fixed Random Seed=true
     )
    )
    Pendulum Pivot:Vector3f=$Particles.Butterfly_Flap_01.Position
    Rigid Pendulum=false
   ) {enabled=false}
   &SolveForcesAndVelocity(Clamp Velocity=false, Limit Acceleration=false)
   &PendulumConstraint() {enabled=false}
   &CalculateAccurateVelocity()
   &ScaleMeshSize(Scale Factor=&VectorFromFloat(Value=&FloatFromCurve(FloatCurve=DataInterface() {propertyValues {Curve {keys [<at -0.0142 -0.0296 RCIM_Cubic>, {arriveTangent=0.0292, interpMode=RCIM_Cubic, leaveTangent=0.0292, time=0.1374, value=0.9863}, <at 1 0>]}}})))
   &UpdateMeshOrientation(
    Rotation Rate=&RandomRangeFloat(Minimum=1, Maximum=2.5)
    Rotation Vector=&RandomVector(Vector Scale=1)
   )
   &ScaleSpriteSize(Scale Factor=&Vector2DFromFloat(Value=&FloatFromCurve(FloatCurve=DataInterface() {propertyValues {Curve {keys [<at 0.0022 0>,<at 0.0574 0>,<at 0.1617 1.0028>,<at 0.3181 0>,<at 0.4315 0>,<at 0.5113 0.9893>,<at 0.5849 0>,<at 0.6861 0>,<at 0.7229 1.0258>,<at 0.7811 0>,<at 1 0>]}}})))
  ]
  rendererClasses [&NiagaraSpriteRendererProperties, &NiagaraMeshRendererProperties]
  renderers [
   {rendererClass=&NiagaraSpriteRendererProperties}
   {rendererClass=&NiagaraMeshRendererProperties, rendererIndex=1}
  ]
 }
 Darken {
  emitterUpdateScript [
   &EmitterState(Inactive Response=::NewEnumerator2, Life Cycle Mode=::NewEnumerator1, Loop Duration=0.5, UseLoopDelay=true)
   &SpawnBurst_Instantaneous(Spawn Count=1, Use Spawn Probability=false)
  ]
  particleSpawnScript [
   /Niagara/Modules/Spawn/Initialization/InitializeParticle(
    Lifetime=1.25
    Module.Write Color=true
    Module.Write Lifetime=true
    Module.Write Mass=true
    Module.Write Material Random=true
    Module.Write Position=false
    Module.Write Scale=false
    Module.Write SpriteRotation=false
    Module.Write SpriteSize=true
    Sprite Size=&Vector2DFromFloat(Value=375)
   )
   &Color(Color=&MakeLinearColorFromVectorAndFloat("Vector (RGB)"=<0.0003 0.0006 0.001>, "Float (Alpha)"=0.8))
  ]
  particleUpdateScript [
   &ParticleState()
   &ScaleColor(
    Scale Alpha=&FloatFromCurve(
     FloatCurve=DataInterface() {propertyValues {Curve {keys [
         <at 0 0 RCIM_Cubic>
         {arriveTangent=-1.1136, interpMode=RCIM_Cubic, leaveTangent=-1.1136, tangentMode=RCTM_Break, time=0.5, value=1}
         <at 1 0>
        ]}}}
     Scale Curve=&FloatFromCurve(FloatCurve=DataInterface() {propertyValues {Curve {keys [<at 50000 1>,<at 72000 0>]}}}, CurveIndex=$Engine.Owner.LODDistance)
    )
    ScaleA=true
    ScaleRGB=true
   )
   /SetVariables(Particles.RenderPosition:Vector3f=&Add_Vector(
     A=Linked() {linkedVariable {name=Particles.Position, type=NiagaraPosition}}
     B=&MakeVector(
      X=0
      Y=0
      Z=&DegreesToRadians(AngleInDegrees:NiagaraFloat=&Sine(Normalized Angle=$System.NormalizedLoopAge, Period=1, Scale=688.9792, Bias=-0.9636))
     )
    )) {isSetParameters=true, module=SetVariables_#3}
   &DynamicMaterialParameters(
    Index 0 Param 1=2
    Index 0 Param 2=1
    Index 0 Param 3=1
    Index 0 Param 4=1
    Module.Param0WriteEnabled=true
    Module.Param1WriteEnabled=true
    Module.Param2WriteEnabled=true
    Module.Param3WriteEnabled=true
   )
   &CameraOffset()
   &ScaleSpriteSize(Scale Factor=&Vector2DFromFloat(Value=&Multiply_Float(
      A=&Lerp_Float(A=1, B=1.55, Alpha=$User.Destabilization)
      B=&FloatFromCurve(FloatCurve=DataInterface() {propertyValues {Curve {keys [<at 40000 1>,<at 55000 0>]}}}, CurveIndex=$Engine.Owner.LODDistance)
     )))
  ]
  rendererClasses [&NiagaraSpriteRendererProperties]
  renderers [{rendererClass=&NiagaraSpriteRendererProperties}]
 }
 FN_ButterflyMesh {
  emitterUpdateScript [
   &EmitterState()
   &SpawnBurst_Instantaneous(Spawn Count=1, Spawn Probability=HlslExpression() {hlslExpression="Emitter.LoopCount == 0"}, Use Spawn Probability=true)
  ]
  particleSpawnScript [
   /Niagara/Modules/Spawn/Initialization/V2/InitializeParticle(
    Color Mode=::NewEnumerator0
    Lifetime=5
    Mesh Scale Mode=::NewEnumerator3
    Mesh Uniform Scale=$User.Scale
    UsePositionOffset=false
    Write Material Random=true
   )
   &InitialMeshOrientation(
    Orientation Vector=<2.9873 1286.4857 1.4349>
    Rotation Coordinate Space=::NewEnumerator1
    Use Orientation Vector=true
    Use Rotation Vector=false
   ) {enabled=false}
   /SetVariables(Particles.Gnomon:Vector3f=<30 30 30>) {isSetParameters=true}
  ]
  particleUpdateScript [
   &ParticleState(Kill Particles When Lifetime Has Elapsed=false)
   /SetVariables(Particles.MeshOrientation=$System.MeshOrientation) {isSetParameters=true}
   /SetVariables(Particles.MeshOrientation=&FindQuatBetweenTwoVectors(From Vector=<1 0 0>, To Vector=<0 1.1043 1>)) {enabled=false, isSetParameters=true}
   &SystemLocation(MaskSpawnGroup=false)
   &FN_Sprite_Attributes(
    Create Position Attribute=true
    Position:Vector3f=&Add_Vector(
     A=Linked() {linkedVariable {name=Particles.Position, type=NiagaraPosition}}
     B=&MakeVector(
      X=0
      Y=0
      Z=&DegreesToRadians(AngleInDegrees:NiagaraFloat=&Sine(Normalized Angle=$Emitter.NormalizedLoopAge, Period=1, Scale=688.9792, Bias=-0.9636))
     )
    )
    Write to Module Namespace=true
   )
   &DynamicMaterialParameters(
    Index 0 Param 1=0.2364
    Index 0 Param 4=$System.NormalizedLoopAge
    Index 1 Param 1=3000
    Index 1 Param 2=1
    Module.Param0WriteEnabled=true
    Module.Param0WriteEnabled1=true
    Module.Param1WriteEnabled=true
    Module.Param1WriteEnabled1=true
    Module.Param2WriteEnabled=true
    Module.Param2WriteEnabled1=true
    Module.Param3WriteEnabled=true
    Module.Param3WriteEnabled1=true
    Write Parameter Index 1=true
   )
  ]
  rendererClasses [&NiagaraSpriteRendererProperties, &NiagaraMeshRendererProperties]
  renderers [
   {rendererClass=&NiagaraSpriteRendererProperties}
   {rendererClass=&NiagaraMeshRendererProperties, rendererIndex=1}
   {rendererClass=&NiagaraMeshRendererProperties, rendererIndex=2}
  ]
 }
 Ribbon {
  emitterSpawnScript [/SetVariables(Emitter.Number of Trails=1) {isSetParameters=true, module=SetVariables_#4}]
  emitterUpdateScript [
   &EmitterState(Life Cycle Mode=::NewEnumerator1, Loop Duration=1, UseLoopDelay=false)
   &SpawnBurst_Instantaneous(
    Spawn Count=$Emitter.Initial.Number of Trails
    Spawn Group=-1
    Spawn Probability=HlslExpression() {hlslExpression="Emitter.LoopCount == 0"}
    Use Spawn Probability=true
   )
   &SpawnRate(
    SpawnRate=&Multiply_Float(A=10, B=&MakeFloatFromInt(Int=$Emitter.Number of Trails))
    Use Spawn Probability=false
   ) {enabled=false}
   &SpawnPerUnit(Module.Use Max Movement Threshold=false, Module.Use Movement Tolerance=true, Movement Tolerance=15, Use Spawn Probability=false) {enabled=false}
   &self:FN_SpawnByDistance(Step=100) {module=FN_SpawnByDistance}
   /Game/Effects/Niagara/Modules/Math/FN_Integrator_v2/FN_Integrator(
    Clamp=true
    Decrement=HlslExpression() {hlslExpression="length(Engine.Owner.Velocity) < 10 ? 1 : 0"}
    Increment=HlslExpression() {hlslExpression="length(Engine.Owner.Velocity) > 10 ? 1 : 0"}
   )
  ]
  particleSpawnScript [
   &FN_SetSpawnGroup(Number of Spawn Groups:NiagaraInt32=$Emitter.Initial.Number of Trails)
   /SetVariables(Particles.SpawnGroup=/Game/Effects/Niagara/DynamicInputs/Arrays/FN_SelectIntFromArray(
     Array Sampling Mode:ENiagaraArraySamplingMode=::NewEnumerator1
     Interpolation:NiagaraFloat=?
     Evaluation Type:ENiagaraRandomnessEvaluation=::NewEnumerator0
     Normalized Interpolation:NiagaraBool=true
     Direct Array Index:NiagaraInt32=$Particles.SpawnGroup
     Recalculate Random Each Loop:NiagaraBool=false
     Int Selection Array:NiagaraDataInterfaceArrayInt32=DataInterface() {propertyValues {IntData [0,1,2,3]}}
     Randomness Mode:ENiagaraRandomnessMode=::NewEnumerator0
     Random Seed:NiagaraInt32=?
     Fixed Random Seed:NiagaraBool=false
    )) {isSetParameters=true, module=SetVariables_#5}
   &self:Ribbon_0.NiagaraScratchPadContainer_1.LockParticlesToBones()
   &InitializeRibbon(
    Lifetime=&MakeCustomFloatFromBool(Bool=$Particles.Locked, True Float=165, False Float=/Niagara/DynamicInputs/Math/Min_Float(A:NiagaraFloat=$Emitter.Age, B:NiagaraFloat=3.5))
    Module.Write Color=true
    Module.Write Lifetime=true
    Module.Write Mass=true
    Module.Write Material Random=true
    Module.Write Position=false
    Module.Write RibbonTwist=false
    Module.Write RibbonWidth=true
    Ribbon Width=150
   )
   &FN_FlipRibbonV()
   &SkeletalMeshLocation(
    Apply To Particle Position=false
    Bone / Socket Index=$Particles.SpawnGroup
    Bone Sampling Mode=::NewEnumerator4
    Inherit Velocity=false
    Interpolate Mesh Sampling=false
    MaskSpawnGroup=false
    Skeletal Mesh=$User.SkeletalMesh
   )
   &AddVelocity(Velocity=<0 0 100>) {enabled=false}
   /SetVariables(Particles.Alpha:NiagaraFloat=$Emitter.FN_Integrator.Value) {isSetParameters=true, module=SetVariables_#6}
  ]
  particleUpdateScript [
   &ParticleState()
   /SetVariables(Particles.Age:NiagaraFloat=&MakeCustomFloatFromBool(Bool=$Particles.Locked, True Float=0, False Float=$Particles.Age)) {isSetParameters=true, module=SetVariables_#7}
   /SetVariables(Particles.NormalizedAge:NiagaraFloat=&MakeCustomFloatFromBool(Bool=$Particles.Locked, True Float=0, False Float=$Particles.NormalizedAge)) {isSetParameters=true, module=SetVariables_#8}
   &SkeletalMeshLocation(
    Apply To Particle Position=false
    Bone / Socket Index=$Particles.SpawnGroup
    Bone Sampling Mode=::NewEnumerator4
    Inherit Velocity=false
    Interpolate Mesh Sampling=false
    MaskSpawnGroup=false
    Skeletal Mesh=$User.SkeletalMesh
   ) {module=SkeletalMeshLocation001}
   &GravityForce(Gravity=<0 0 -1>)
   &CurlNoiseForce(Noise Frequency=2, Noise Strength=4, Pan Noise=true, Pan Noise Field=<0 0 1>)
   &SolveForcesAndVelocity(Clamp Velocity=false, Limit Acceleration=false)
   /SetVariables(Particles.RenderPosition:Vector3f=&MakeCustomVectorFromBool(
     Bool=$Particles.Locked
     True Vector=$Output.SkeletalMeshLocation001.SampledPosition
     False Vector=&MakeCustomVectorFromBool(
      Bool=$Emitter.LocalSpace
      True Vector=&Add_Vector(A=$Output.SkeletalMeshLocation001.SampledPosition, B=Linked() {linkedVariable {name=Particles.Position, type=NiagaraPosition}})
      False Vector=Linked() {linkedVariable {name=Particles.Position, type=NiagaraPosition}}
     )
    )) {isSetParameters=true, module=SetVariables_#9}
   &FN_RibbonMaterial(
    Color=<rgba 0.0541 0.0765 0.099 1>
    Erosion Amount=0.4455
    Erosion Pan Speed U=0.1
    Erossion Scale U=0.1091
    Opacity Pan U Speed=0.1
    Opacity Scale U=0.2273
    Scale Alpha:NiagaraFloat=&FloatFromCurve(
     FloatCurve=DataInterface() {propertyValues {Curve {keys [<at 0 1>,<at 1 0.0003>]}}}
     Scale Curve=&Multiply_Float(A=2, B=$Particles.Alpha)
    )
   )
   /SetVariables(Particles.RibbonFacing=<0 0 1>) {isSetParameters=true, module=SetVariables_#10}
   &KillParticlesInVolume(
    Invert Volume=true
    Kill Volume Enabled:NiagaraBool=&Boolean_NotOperation(Bool=$Particles.Locked)
    Sphere Radius=10000
    Volume Origin:NiagaraPosition=$Engine.Owner.Position
   )
  ]
  rendererClasses [&NiagaraRibbonRendererProperties, &NiagaraSpriteRendererProperties]
  renderers [
   {rendererClass=&NiagaraRibbonRendererProperties}
   {rendererClass=&NiagaraSpriteRendererProperties, rendererIndex=1}
  ]
 }
 Ribbon001 {
  emitterSpawnScript [/SetVariables(Emitter.Number of Trails=1) {isSetParameters=true, module=SetVariables_#4}]
  emitterUpdateScript [
   &EmitterState(Life Cycle Mode=::NewEnumerator1, Loop Duration=1, UseLoopDelay=false)
   &SpawnBurst_Instantaneous(
    Spawn Count=$Emitter.Initial.Number of Trails
    Spawn Group=-1
    Spawn Probability=HlslExpression() {hlslExpression="Emitter.LoopCount == 0"}
    Use Spawn Probability=true
   )
   &SpawnRate(
    SpawnRate=&Multiply_Float(A=10, B=&MakeFloatFromInt(Int=$Emitter.Number of Trails))
    Use Spawn Probability=false
   ) {enabled=false}
   &SpawnPerUnit(Module.Use Max Movement Threshold=false, Module.Use Movement Tolerance=true, Movement Tolerance=15, Use Spawn Probability=false) {enabled=false}
   &self:FN_SpawnByDistance(Step=100) {module=FN_SpawnByDistance}
   &FN_Integrator(
    Clamp=true
    Decrement=HlslExpression() {hlslExpression="length(Engine.Owner.Velocity) < 10 ? 1 : 0"}
    Increment=HlslExpression() {hlslExpression="length(Engine.Owner.Velocity) > 10 ? 1 : 0"}
   )
  ]
  particleSpawnScript [
   &FN_SetSpawnGroup(Number of Spawn Groups:NiagaraInt32=$Emitter.Initial.Number of Trails)
   /SetVariables(Particles.SpawnGroup=&FN_SelectIntFromArray(
     Array Sampling Mode:ENiagaraArraySamplingMode=::NewEnumerator1
     Interpolation:NiagaraFloat=?
     Evaluation Type:ENiagaraRandomnessEvaluation=::NewEnumerator0
     Normalized Interpolation:NiagaraBool=true
     Direct Array Index:NiagaraInt32=$Particles.SpawnGroup
     Recalculate Random Each Loop:NiagaraBool=false
     Int Selection Array:NiagaraDataInterfaceArrayInt32=DataInterface() {propertyValues {IntData [0,1,2,3]}}
     Randomness Mode:ENiagaraRandomnessMode=::NewEnumerator0
     Random Seed:NiagaraInt32=?
     Fixed Random Seed:NiagaraBool=false
    )) {isSetParameters=true, module=SetVariables_#5}
   &self:Ribbon001_1.NiagaraScratchPadContainer_1.LockParticlesToBones()
   &InitializeRibbon(
    Lifetime=&MakeCustomFloatFromBool(Bool=$Particles.Locked, True Float=165, False Float=&Min_Float(A:NiagaraFloat=$Emitter.Age, B:NiagaraFloat=2.2))
    Module.Write Color=true
    Module.Write Lifetime=true
    Module.Write Mass=true
    Module.Write Material Random=true
    Module.Write Position=false
    Module.Write RibbonTwist=false
    Module.Write RibbonWidth=true
    Ribbon Width=60
   )
   &FN_FlipRibbonV()
   &SkeletalMeshLocation(
    Apply To Particle Position=false
    Bone / Socket Index=$Particles.SpawnGroup
    Bone Sampling Mode=::NewEnumerator4
    Inherit Velocity=false
    Interpolate Mesh Sampling=false
    MaskSpawnGroup=false
    Skeletal Mesh=$User.SkeletalMesh
   )
   &AddVelocity(Velocity=<0 0 100>) {enabled=false}
   /SetVariables(Particles.Alpha:NiagaraFloat=$Emitter.FN_Integrator.Value) {isSetParameters=true, module=SetVariables_#6}
  ]
  particleUpdateScript [
   &ParticleState()
   /SetVariables(Particles.Age:NiagaraFloat=&MakeCustomFloatFromBool(Bool=$Particles.Locked, True Float=0, False Float=$Particles.Age)) {isSetParameters=true, module=SetVariables_#7}
   /SetVariables(Particles.NormalizedAge:NiagaraFloat=&MakeCustomFloatFromBool(Bool=$Particles.Locked, True Float=0, False Float=$Particles.NormalizedAge)) {isSetParameters=true, module=SetVariables_#8}
   &SkeletalMeshLocation(
    Apply To Particle Position=false
    Bone / Socket Index=$Particles.SpawnGroup
    Bone Sampling Mode=::NewEnumerator4
    Inherit Velocity=false
    Interpolate Mesh Sampling=false
    MaskSpawnGroup=false
    Skeletal Mesh=$User.SkeletalMesh
   ) {module=SkeletalMeshLocation001}
   &GravityForce(Gravity=<0 0 -1>)
   &CurlNoiseForce(Noise Frequency=2, Noise Strength=4, Pan Noise=true, Pan Noise Field=<0 0 1>)
   &SolveForcesAndVelocity(Clamp Velocity=false, Limit Acceleration=false)
   /SetVariables(Particles.RenderPosition:Vector3f=&MakeCustomVectorFromBool(
     Bool=$Particles.Locked
     True Vector=$Output.SkeletalMeshLocation001.SampledPosition
     False Vector=&MakeCustomVectorFromBool(
      Bool=$Emitter.LocalSpace
      True Vector=&Add_Vector(A=$Output.SkeletalMeshLocation001.SampledPosition, B=Linked() {linkedVariable {name=Particles.Position, type=NiagaraPosition}})
      False Vector=Linked() {linkedVariable {name=Particles.Position, type=NiagaraPosition}}
     )
    )) {isSetParameters=true, module=SetVariables_#9}
   &FN_RibbonMaterial(
    Color=<rgba 0.5326 0.7533 0.974 1>
    Erosion Amount=0.4455
    Erosion Pan Speed U=0.1
    Erossion Scale U=0.1091
    Opacity Pan U Speed=0.1
    Opacity Scale U=0.2273
    Scale Alpha:NiagaraFloat=&FloatFromCurve(
     FloatCurve=DataInterface() {propertyValues {Curve {keys [<at 0 1>,<at 1 0.0003>]}}}
     Scale Curve=&Multiply_Float(A=1, B=$Particles.Alpha)
    )
   )
   /SetVariables(Particles.RibbonFacing=<0 0 1>) {isSetParameters=true, module=SetVariables_#10}
   &KillParticlesInVolume(
    Invert Volume=true
    Kill Volume Enabled:NiagaraBool=&Boolean_NotOperation(Bool=$Particles.Locked)
    Sphere Radius=10000
    Volume Origin:NiagaraPosition=$Engine.Owner.Position
   )
  ]
  rendererClasses [&NiagaraRibbonRendererProperties, &NiagaraSpriteRendererProperties]
  renderers [
   {rendererClass=&NiagaraRibbonRendererProperties}
   {rendererClass=&NiagaraSpriteRendererProperties, rendererIndex=1}
  ]
 }
 ZeroPointFlare {
  emitterUpdateScript [
   &EmitterState()
   &SpawnBurst_Instantaneous(Spawn Count=1, Use Spawn Probability=false) {enabled=false}
   &SpawnRate(Use Spawn Probability=false)
   /SetVariables(Emitter.DistanceRamp:NiagaraFloat=/Game/Effects/Niagara/DynamicInputs/Math/FN_Fit(
     Clamp:NiagaraBool=true
     Value:NiagaraFloat=$Engine.Owner.LODDistance
     SourceMin:NiagaraFloat=300
     SourceMax:NiagaraFloat=1000
     DestMin:NiagaraFloat=0.2
     DestMax:NiagaraFloat=1
    )) {isSetParameters=true}
  ]
  particleSpawnScript [
   /Niagara/Modules/Spawn/Initialization/InitializeParticle(
    Lifetime=&RandomRangeFloat(Minimum=1.8, Maximum=1.45)
    Module.Write Color=true
    Module.Write Lifetime=true
    Module.Write Mass=true
    Module.Write Material Random=true
    Module.Write Position=false
    Module.Write Scale=false
    Module.Write SpriteRotation=true
    Module.Write SpriteSize=true
    Sprite Rotation=&RandomRangeFloat(Minimum=0, Maximum=360)
    Sprite Size=<3000 3000>
   )
   &Color(Color=&MakeLinearColorFromVectorAndFloat("Vector (RGB)"=<0.0175 0.0243 0.03>, "Float (Alpha)"=&ClampFloat(Float=0.6, Max=0.5)))
  ]
  particleUpdateScript [
   &ParticleState()
   &ScaleColor(
    Scale Alpha=&FloatFromCurve(FloatCurve=DataInterface() {propertyValues {Curve {keys [<at 0 1 RCIM_Cubic>,<at 0.25 1 RCIM_Cubic>,<at 0.75 1 RCIM_Cubic>,<at 1 1>]}}})
    ScaleA=true
    ScaleRGB=true
   )
   /SetVariables(Particles.Size:NiagaraFloat=&FloatFromCurve(
     FloatCurve=DataInterface() {propertyValues {Curve {keys [
         <at -0.0039 0.0032 RCIM_Cubic>
         {arriveTangent=0.0358, interpMode=RCIM_Cubic, leaveTangent=0.0358, tangentMode=RCTM_User, time=0.4967, value=0.9107}
         <at 1 0>
        ]}}}
     Scale Curve=&Lerp_Float(A=0, B=1, Alpha=$Emitter.DistanceRamp)
    )) {isSetParameters=true}
   /SetVariables(Particles.RenderPosition:Vector3f=&Add_Vector(
     A=Linked() {linkedVariable {name=Particles.Position, type=NiagaraPosition}}
     B=&MakeVector(
      X=0
      Y=0
      Z=&DegreesToRadians(AngleInDegrees:NiagaraFloat=&Sine(Normalized Angle=$System.NormalizedLoopAge, Period=1, Scale=688.9792, Bias=-0.9636))
     )
    )) {isSetParameters=true, module=SetVariables_#3}
   &FN_Sprite_Attributes(
    Camera Offset=50
    Color=&MakeLinearColorFromVectorAndFloat(
     "Vector (RGB)"=&Multiply_VectorByFloat(Vector=<0.5625 1.8964 3>, Float=0.5)
     "Float (Alpha)"=&Multiply_Float(A=$Particles.Size, B=2)
    )
    Create Camera Offset Attribute=true
    Create Color Attribute=true
    Create SpriteSize Attribute=true
    SpriteSize=&Vector2DFromFloat(Value=&Multiply_Float(A=&Multiply_Float(A=800, B=$Emitter.DistanceRamp), B=$Particles.Size))
    Write to Module Namespace=true
   ) {module=FN_Sprite_Attributes001}
   &FN_Sprite_Attributes(
    Camera Offset=50
    Color=&MakeLinearColorFromVectorAndFloat(
     "Vector (RGB)"=<0.8602 1.4152 2>
     "Float (Alpha)"=&FloatFromCurve(
      FloatCurve=DataInterface() {propertyValues {Curve {keys [
          <at 0.0508 -0.0195 RCIM_Cubic>
          {arriveTangent=-0.0652, interpMode=RCIM_Cubic, leaveTangent=-0.0652, tangentMode=RCTM_User, time=0.4875, value=0.1732}
          <at 0.9532 -0.0003>
         ]}}}
      Scale Curve=0.3
     )
    )
    Create Camera Offset Attribute=true
    Create Color Attribute=true
    Create SpriteSize Attribute=true
    SpriteSize=&Vector2DFromFloat(Value=&Lerp_Float(A=100, B=200, Alpha=$Emitter.DistanceRamp))
    Write to Module Namespace=true
   ) {module=FN_Sprite_Attributes002}
   &FN_Sprite_Attributes(
    Color=&MakeLinearColorFromVectorAndFloat(
     "Vector (RGB)"=<1.5665 2.3369 6>
     "Float (Alpha)"=&FloatFromCurve(
      FloatCurve=DataInterface() {propertyValues {Curve {keys [
          <at 0.0048 -0.0024 RCIM_Cubic>
          {arriveTangent=2.9233, interpMode=RCIM_Cubic, leaveTangent=2.9233, tangentMode=RCTM_User, time=0.7087, value=0.5852}
          <at 0.9988 -0.0024>
         ]}}}
      Scale Curve=0.2
     )
    )
    Create Color Attribute=true
    Create Sprite Rotation Attribute=true
    Create SpriteSize Attribute=true
    SpriteSize=&Multiply_Vector2(
     A=<1000 20>
     B=&Vector2DFromFloat(Value=&Multiply_Float(A=$Emitter.DistanceRamp, B=&RandomRangeFloat(Minimum=0.2, Maximum=0.5)))
    )
    Write to Module Namespace=true
   ) {module=FN_Sprite_Attributes003}
   &FN_Sprite_Attributes(
    Color=&MakeLinearColorFromVectorAndFloat("Vector (RGB)"=<0.2398 1.0892 1.5>, "Float (Alpha)"=0.2)
    Create Color Attribute=true
    Create Sprite Rotation Attribute=true
    Create SpriteSize Attribute=true
    SpriteSize=&Multiply_Vector2(
     A=&RandomRangeVector2D(Minimum=<120 80>, Maximum=<200 140>)
     B=&Vector2DFromFloat(Value=&FloatFromCurve(
       FloatCurve=DataInterface() {propertyValues {Curve {keys [<at 0 0 RCIM_Cubic>,<at 0.5154 0.7022 RCIM_Cubic>,<at 1 0>]}}}
       Scale Curve=$Emitter.DistanceRamp
      ))
    )
    Write to Module Namespace=true
   ) {module=FN_Sprite_Attributes004}
  ]
  rendererClasses [&NiagaraSpriteRendererProperties]
  renderers [
   {rendererClass=&NiagaraSpriteRendererProperties}
   {rendererClass=&NiagaraSpriteRendererProperties, rendererIndex=1}
   {rendererClass=&NiagaraSpriteRendererProperties, rendererIndex=2}
   {rendererClass=&NiagaraSpriteRendererProperties, rendererIndex=3}
  ]
 }
}
summary {
 systemName=NS_JFly
 emitters [
  {
   emitterName=FN_ButterflyMesh
   rendererClasses [&NiagaraSpriteRendererProperties, &NiagaraMeshRendererProperties]
  }
  {emitterName=Audio_EMitter, rendererClasses []}
  {emitterName=Ribbon, rendererClasses [&NiagaraRibbonRendererProperties, &NiagaraSpriteRendererProperties]}
  {emitterName=Ribbon001, rendererClasses [&NiagaraRibbonRendererProperties, &NiagaraSpriteRendererProperties]}
  {emitterName=ZeroPointFlare, rendererClasses [&NiagaraSpriteRendererProperties]}
  {
   emitterName=Crystals
   simTarget=GPUComputeSim
   rendererClasses [&NiagaraSpriteRendererProperties, &NiagaraMeshRendererProperties]
  }
  {emitterName=Darken, rendererClasses [&NiagaraSpriteRendererProperties]}
 ]
 userVariables [
  User.Destabilization {type=NiagaraFloat, defaultValue=0, description=""}
  User.Falloff {type=NiagaraFloat, defaultValue=867.8375, description=""}
  User.Scale {type=NiagaraFloat, defaultValue=0.1, description=""}
  User.SkeletalMesh {
   type=NiagaraDataInterfaceSkeletalMesh
   defaultValue {
    struct=NiagaraExt_VariableValue_DataInterface
    value {dataInterface=&self:NiagaraDataInterfaceSkeletalMesh_0}
   }
  }
 ]
}
