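# Sound cue for the Season 16 butterfly shard teleport, playing on the quest sound class.
# A random node picks one of three equally weighted Butterfly_Collect waves, reused from the S16 narrative quest sounds, with raised pitch and boosted volume.
# It references the narrative familiar attenuation asset and overrides concurrency to cap simultaneous instances.
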SoundCue /BattlepassS16/Sounds/Shard/Butterfly_Shard_Teleport_Cue
root=Random_0
nodes [
 {id=Random_0, children [WavePlayer_1,WavePlayer_3,WavePlayer_4], properties {weights [1,1,1]}}
 {id=WavePlayer_1, properties {soundWaveAssetPtr=/Game/Athena/Sounds/Quests/S16_Narrative/Butterfly_Collect_01}}
 {id=WavePlayer_3, properties {soundWaveAssetPtr=/Game/Athena/Sounds/Quests/S16_Narrative/Butterfly_Collect_02}}
 {id=WavePlayer_4, properties {soundWaveAssetPtr=/Game/Athena/Sounds/Quests/S16_Narrative/Butterfly_Collect_03}}
]
properties {
 attenuationSettings=/Game/Athena/Sounds/Quests/S16_Narrative/S16_Narrative_Familiar_Att
 bOverrideConcurrency=true
 maxDistance=3512
 pitchMultiplier=1.2
 soundClassObject=&Fort_Quests
 volumeMultiplier=4.48
 attenuationOverrides {falloffDistance=3000, attenuationShapeExtents {x=700}}
 concurrencyOverrides {maxCount=3}
}
