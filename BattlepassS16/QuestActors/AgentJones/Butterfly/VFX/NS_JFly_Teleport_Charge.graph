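# Niagara charge-up effect for the Agent Jones Butterfly teleport in the Battlepass S16 quest actors, building over roughly two seconds.
# Mesh cracks, light beams and electric arcs combine with GPU embers collapsing along shrinking random tori, shrinking gas sprites and a layered, pulsing flare with lights.
# Curves indexed by emitter age ramp spawn rates, scales and brightness while tightening radii.
# Crack and beam lifetimes subtract emitter age so they end together.
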NiagaraSystem /BattlepassS16/QuestActors/AgentJones/Butterfly/VFX/NS_JFly_Teleport_Charge
dependencies {
 usedDataInterfaces [
  &NiagaraDataInterfaceCurve
  &NiagaraDataInterfaceVectorCurve
  &NiagaraDataInterfaceVector2DCurve
 ]
 usedDynamicInputs []
 usedModules [&SystemState]
}
emitters {
 Crack {
  emitterUpdateScript [
   &EmitterState()
   &SpawnBurst_Instantaneous(Spawn Count=1, Spawn Time=0.7875, Use Spawn Probability=false) {enabled=false}
   &SpawnBurst_Instantaneous(Spawn Count=1, Spawn Time=0.8708, Use Spawn Probability=false) {enabled=false, module=SpawnBurst_Instantaneous001}
   &SpawnBurst_Instantaneous(Spawn Count=1, Spawn Time=1.175, Use Spawn Probability=false) {enabled=false, module=SpawnBurst_Instantaneous002}
   &SpawnRate(SpawnRate=3, Use Spawn Probability=false)
  ]
  particleSpawnScript [
   /Niagara/Modules/Spawn/Initialization/InitializeParticle(
    Lifetime=&Subtract_Float(A=2, B=$Emitter.Age)
    Mesh Scale=&VectorFromFloat(Value=&RandomRangeFloat(Minimum=2, Maximum=4))
    Module.Write Color=true
    Module.Write Lifetime=true
    Module.Write Mass=true
    Module.Write Material Random=true
    Module.Write Position=false
    Module.Write Scale=true
    Module.Write SpriteRotation=false
    Module.Write SpriteSize=false
   )
   &InitialMeshOrientation(Orientation Vector=&RandomVector(Vector Scale=1), Use Orientation Vector=true, Use Rotation Vector=false)
   &Color(Color=&MakeLinearColorFromVectorAndFloat("Vector (RGB)"=<3 3 3>, "Float (Alpha)"=&ClampFloat(Float=0.25)))
  ]
  particleUpdateScript [
   &ParticleState()
   &DynamicMaterialParameters(
    Index 0 Param 1=&FloatFromCurve(FloatCurve=DataInterface() {propertyValues {Curve {keys [
         <at -0.0018 0.157 RCIM_Constant>
         <at 0.314 0.2679 RCIM_Constant>
         <at 0.4109 0.3524 RCIM_Constant>
         <at 0.5034 0.4332 RCIM_Constant>
         <at 0.5551 0.4782 RCIM_Constant>
         <at 0.7355 0.6357 RCIM_Constant>
         {interpMode=RCIM_Constant, leaveTangent=1.4348, time=1, value=1}
        ]}}})
    Index 0 Param 2=&FloatFromCurve(
     FloatCurve=DataInterface() {propertyValues {Curve {keys [
         {interpMode=RCIM_Cubic, tangentMode=RCTM_User, time=0.0035, value=0.385}
         {arriveTangent=0.5933, interpMode=RCIM_Cubic, leaveTangent=0.5933, tangentMode=RCTM_User, time=0.1535, value=0.4764}
         <at 0.432 0.6074 RCIM_Constant>
         <at 0.5695 0.704 RCIM_Constant>
         <at 0.7241 0.8126 RCIM_Constant>
         {interpMode=RCIM_Cubic, leaveTangent=2.1533, tangentMode=RCTM_User, time=1, value=1}
        ]}}}
     Scale Curve=0.5
    )
    Index 0 Param 3=&FloatFromCurve(
     FloatCurve=DataInterface() {propertyValues {Curve {keys [
         {arriveTangent=0.132, time=0.5032, value=-0.1262}
         {arriveTangent=0.2221, leaveTangent=0.2221, time=1.0077, value=0.0321}
         <at 2.1371 0.0969>
        ]}}}
     Scale Curve=0.1
    )
    Module.Param0WriteEnabled=true
    Module.Param1WriteEnabled=true
    Module.Param2WriteEnabled=true
    Module.Param3WriteEnabled=true
   )
  ]
  rendererClasses [&NiagaraMeshRendererProperties]
  renderers [{rendererClass=&NiagaraMeshRendererProperties, rendererIndex=0}]
 }
 ElectricArcsResidual {
  emitterUpdateScript [
   &EmitterState()
   &SpawnRate(
    SpawnRate=&FloatFromCurve(
     FloatCurve=DataInterface() {propertyValues {Curve {keys [
         <at -0.011 0.0168 RCIM_Cubic>
         {arriveTangent=1.5569, interpMode=RCIM_Cubic, leaveTangent=1.5569, tangentMode=RCTM_User, time=2.0041, value=1.9992}
        ]}}}
     CurveIndex=$Emitter.NormalizedLoopAge
     Scale Curve=60
    )
    Use Spawn Probability=false
   )
  ]
  particleSpawnScript [
   /SetVariables(Particles.MeshOrientation=&MakeQuatFromAxisAngle(
     Axis=&RandomRangeVector(Minimum=<0 0 0>, Maximum=<1 0 0>, Evaluation Type=::NewEnumerator1)
     AngleInDegrees=&RandomRangeFloat(Minimum=0, Maximum=360)
    )) {isSetParameters=true}
   &InitializeParticle(
    Lifetime=&RandomRangeFloat(Minimum=0.08, Maximum=0.2, Evaluation Type=::NewEnumerator1)
    Mesh Scale=&Multiply_VectorByFloat(
     Vector=&RandomRangeVector(Minimum=<8 6 6>, Maximum=<15 20 20>, Evaluation Type=::NewEnumerator1)
     Float=&FloatFromCurve(
      FloatCurve=DataInterface() {propertyValues {Curve {keys [
          <at -0.011 0.0168 RCIM_Cubic>
          {arriveTangent=1.5569, interpMode=RCIM_Cubic, leaveTangent=1.5569, tangentMode=RCTM_User, time=2.0041, value=1.9992}
         ]}}}
      CurveIndex=$Emitter.NormalizedLoopAge
      Scale Curve=2
     )
    )
    Module.Write Color=true
    Module.Write Lifetime=true
    Module.Write Mass=true
    Module.Write Material Random=true
    Module.Write Position=false
    Module.Write Scale=true
    Module.Write SpriteRotation=false
    Module.Write SpriteSize=false
   )
   &DynamicMaterialParameters(
    Index 0 Param 1=&RandomRangeFloat(Minimum=2, Maximum=3, Evaluation Type=::NewEnumerator1)
    Index 0 Param 3=&RandomRangeFloat(Minimum=0.2, Maximum=2, Evaluation Type=::NewEnumerator1)
    Index 0 Param 4=&RandomRangeFloat(Minimum=0, Maximum=1, Evaluation Type=::NewEnumerator1)
    Module.Param0WriteEnabled=true
    Module.Param1WriteEnabled=true
    Module.Param2WriteEnabled=true
    Module.Param3WriteEnabled=true
   )
   &Color(Color=&MakeLinearColorFromVectorAndFloat("Vector (RGB)"=<0.0986 1 0.6632>, "Float (Alpha)"=&ClampFloat(Float=1)))
   /Niagara/Modules/Spawn/Location/SphereLocation(MaskSpawnGroup=false, Module.Override Randomness=false, Sphere Radius=10, Surface Only=true)
   &AddVelocityFromPoint(Module.Use Falloff Distance=false)
  ]
  particleUpdateScript [
   &ParticleState()
   &DynamicMaterialParameters(
    Index 0 Param 2=&FloatFromCurve(FloatCurve=DataInterface() {propertyValues {Curve {keys [{arriveTangent=-1.8894, interpMode=RCIM_Cubic, time=0, value=1}, <at 1 -0.25>]}}})
    Module.Param0WriteEnabled=true
    Module.Param1WriteEnabled=true
    Module.Param2WriteEnabled=true
    Module.Param3WriteEnabled=true
   ) {module=DynamicMaterialParameters001}
   &ScaleColor(
    Scale RGB=&VectorFromCurve(
     VectorCurve=DataInterface() {propertyValues {
       XCurve {keys [<at 0 50 RCIM_Cubic>, {arriveTangent=-45, interpMode=RCIM_Cubic, leaveTangent=-45, time=0.5, value=30}, <at 1 5>]}
       YCurve {keys [<at 0 19.1123 RCIM_Cubic>, {arriveTangent=-18.2463, interpMode=RCIM_Cubic, leaveTangent=-18.2463, time=0.5, value=13.6826}, <at 1 0.8661>]}
       ZCurve {keys [<at 0 22.1016 RCIM_Cubic>, {arriveTangent=-21.4843, interpMode=RCIM_Cubic, leaveTangent=-21.4843, time=0.5, value=12.787}, <at 1 0.6173>]}
      }}
     Scale Curve=&VectorFromFloat(Value=&FloatFromCurve(FloatCurve=DataInterface() {propertyValues {Curve {keys [<at 0 1>,<at 2.0385 4.1157>]}}}, CurveIndex=$Emitter.Age, Scale Curve=2))
    )
    ScaleA=true
    ScaleRGB=true
   )
   &ScaleMeshSize(Scale Factor=&VectorFromCurve(VectorCurve=DataInterface() {propertyValues {
       XCurve {keys [<at 0.0033 0.5 RCIM_Constant>,<at 0.2054 0.7619 RCIM_Constant>,<at 0.4101 0.8575 RCIM_Constant>,<at 0.6967 1>]}
       YCurve {keys [<at 0.0033 0.6004 RCIM_Constant>,<at 0.2054 0.6716 RCIM_Constant>,<at 0.4101 0.3646 RCIM_Constant>,<at 0.6967 0.2468>]}
       ZCurve {keys [<at 0.0033 0.7767 RCIM_Constant>,<at 0.2054 0.5215 RCIM_Constant>,<at 0.4101 0.4457 RCIM_Constant>,<at 0.6967 0.6282>]}
      }}))
   &ScaleMeshSize(Scale Factor=&VectorFromFloat(Value=&FloatFromCurve(FloatCurve=DataInterface() {propertyValues {Curve {keys [<at 0 1>,<at 15000 1.785>]}}}, CurveIndex=$Engine.Owner.LODDistance))) {module=ScaleMeshSize001}
   &CameraOffset(Camera Offset Amount=100)
   &SolveForcesAndVelocity(Clamp Velocity=false, Limit Acceleration=false)
  ]
  rendererClasses [&NiagaraMeshRendererProperties]
  renderers [{rendererClass=&NiagaraMeshRendererProperties, rendererIndex=0}]
 }
 Embers {
  simTarget=GPUComputeSim
  emitterUpdateScript [
   &EmitterState()
   &SpawnRate(
    SpawnRate=&FloatFromCurve(
     FloatCurve=DataInterface() {propertyValues {Curve {keys [
         <at 0.7457 0.1051 RCIM_Cubic>
         {arriveTangent=2.1165, interpMode=RCIM_Cubic, leaveTangent=2.1165, tangentMode=RCTM_User, time=1.9973, value=1.9443}
        ]}}}
     CurveIndex=$Emitter.Age
     Scale Curve=600
    )
    Use Spawn Probability=false
   )
  ]
  particleSpawnScript [
   &InitializeParticle(
    Lifetime=&Multiply_Float(
     A=&RandomRangeFloat(Minimum=1, Maximum=2, Evaluation Type=::NewEnumerator1)
     B=&FloatFromCurve(
      FloatCurve=DataInterface() {propertyValues {Curve {keys [<at 0 1 RCIM_Cubic>, {arriveTangent=-1.8914, interpMode=RCIM_Cubic, leaveTangent=-1.8914, tangentMode=RCTM_User, time=1.9864, value=0.0115}]}}}
      CurveIndex=$Emitter.Age
     )
    )
    Module.Write Color=true
    Module.Write Lifetime=true
    Module.Write Mass=true
    Module.Write Material Random=true
    Module.Write Position=false
    Module.Write Scale=false
    Module.Write SpriteRotation=true
    Module.Write SpriteSize=true
    Sprite Rotation=&NormalizedAngleToDegrees(NormalizedAngle=&RandomRangeFloat(Minimum=0, Maximum=1, Evaluation Type=::NewEnumerator1))
    Sprite Size=<2 4>
   )
   /SetVariables(Particles.ScaleU:NiagaraFloat=&FloatFromCurve(
     FloatCurve=DataInterface() {propertyValues {Curve {keys [
         {arriveTangent=-1.6504, interpMode=RCIM_Cubic, leaveTangent=-1.6504, tangentMode=RCTM_User, time=-0.02, value=1.0266}
         <at 0.4682 0.4365>
         {arriveTangent=-2.0092, interpMode=RCIM_Cubic, leaveTangent=-2.0092, tangentMode=RCTM_User, time=2.0066, value=0.0712}
        ]}}}
     CurveIndex=$Emitter.Age
    )) {isSetParameters=true}
   &Color(Color=&MakeLinearColorFromVectorAndFloat("Vector (RGB)"=<0.6157 1 2>, "Float (Alpha)"=&Multiply_Float(A=1, B=$Particles.ScaleU)))
   &AddVelocity(Velocity=&RandomRangeVector(Minimum=<0 0 -12>, Maximum=<0 0 -5>, Evaluation Type=::NewEnumerator1)) {enabled=false}
   &SphereLocation(
    MaskSpawnGroup=false
    Module.Override Randomness=false
    Sphere Origin:NiagaraPosition=$Engine.Owner.Position
    Sphere Radius=&Multiply_Float(
     A=&RandomRangeFloat(Minimum=100, Maximum=100)
     B=&FloatFromCurve(FloatCurve=DataInterface() {propertyValues {Curve {keys [<at 0 1>,<at 1 0>]}}})
    )
    Surface Only=true
   ) {enabled=false}
  ]
  particleUpdateScript [
   &ParticleState()
   &TorusLocation(
    Align Torus With Axis=true
    Handle Radius=0
    Large Radius:NiagaraFloat=&FloatFromCurve(
     FloatCurve=DataInterface() {propertyValues {Curve {keys [{arriveTangent=-2.8702, interpMode=RCIM_Cubic, leaveTangent=-2.8702, tangentMode=RCTM_User, time=0, value=1}, <at 1 0>]}}}
     Scale Curve=&FloatFromCurve(
      FloatCurve=DataInterface() {propertyValues {Curve {keys [<at -0.0318 1 RCIM_Cubic>, {arriveTangent=-0.688, interpMode=RCIM_Cubic, leaveTangent=-0.688, tangentMode=RCTM_User, time=2.0112, value=0.3}]}}}
      CurveIndex=$Emitter.Age
      Scale Curve=&RandomRangeFloat(Minimum=100, Maximum=300)
     )
    )
    MaskSpawnGroup=false
    Radius Position=0
    Torus Axis=&RandomVector(Vector Scale=1)
    Torus Distribution Mode:ENiagaraTorusDistributionMode=::NewEnumerator1
    Torus Origin:NiagaraPosition=$Engine.Owner.Position
    U Position:NiagaraFloat=&Add_Float(
     A=&RandomRangeFloat(Minimum=0, Maximum=1)
     B=&Multiply_Float(
      A=&FloatFromCurve(FloatCurve=DataInterface() {propertyValues {Curve {keys [<at 0 1>,<at 1 0>]}}}, Scale Curve=0.5)
      B=$Particles.Initial.ScaleU
     )
    )
    V Position=1.3106
   )
   &CalculateAccurateVelocity()
   &ScaleColor(
    Scale Alpha=&FloatFromCurve(FloatCurve=DataInterface() {propertyValues {Curve {keys [<at 0 0 RCIM_Cubic>,<at 0.05 1>,<at 0.1803 1>,<at 0.2169 0.0158>]}}})
    Scale RGB=&VectorFromCurve(VectorCurve=DataInterface() {propertyValues {
       XCurve {keys [<at 0.2073 946.7266 RCIM_Cubic>,<at 0.8 0>]}
       YCurve {keys [<at 0.2073 946.7266 RCIM_Cubic>,<at 0.8 0>]}
       ZCurve {keys [<at 0.2073 946.7266 RCIM_Cubic>,<at 0.8 0>]}
      }})
    ScaleA=true
    ScaleRGB=true
   )
   &ScaleSpriteSize(Scale Factor=&Multiply_Vector2(
     A=&Vector2DFromCurve(
      Vector2Curve=DataInterface() {propertyValues {XCurve {keys [<at 0 2>,<at 0.8728 -0.1931>]}, YCurve {keys [<at 0.0009 3.4494>,<at 0.7495 2.2926>,<at 0.8728 -0.1931>]}}}
      ScaleCurve=&MakeVector2D(X=$Particles.ScaleU, Y=1)
     )
     B=&Vector2DFromFloat(Value=1)
    ))
   &FN_Sprite_Attributes(Color=<rgba 100 1 20 1>, Create Color Attribute=true, Write to Module Namespace=true) {enabled=false}
  ]
  rendererClasses [&NiagaraSpriteRendererProperties]
  renderers [{rendererClass=&NiagaraSpriteRendererProperties}]
 }
 LightbeamsWarmup {
  emitterUpdateScript [
   &EmitterState()
   &SpawnBurst_Instantaneous(Spawn Count=4, Use Spawn Probability=false) {enabled=false}
   &SpawnRate(Use Spawn Probability=false)
  ]
  particleSpawnScript [
   /SetVariables(Particles.MeshOrientation=&MakeQuatFromAxisAngle(
     Axis=&RandomRangeVector(Minimum=<0 0 0>, Maximum=<1 1 1>, Evaluation Type=::NewEnumerator1)
     AngleInDegrees=90
    )) {isSetParameters=true}
   /SetVariables(Particles.SpawnRotationVector=&RandomRangeVector(Minimum=<0 0 0>, Maximum=<1 1 1>)) {isSetParameters=true}
   &InitializeParticle(
    Lifetime=&Subtract_Float(A=2, B=$Emitter.Age)
    Mesh Scale=&VectorFromFloat(Value=&RandomRangeFloat(Minimum=2, Maximum=5, Evaluation Type=::NewEnumerator1))
    Module.Write Color=true
    Module.Write Lifetime=true
    Module.Write Mass=true
    Module.Write Material Random=true
    Module.Write Position=false
    Module.Write Scale=true
    Module.Write SpriteRotation=false
    Module.Write SpriteSize=false
   )
   &Color(Color=&MakeLinearColorFromVectorAndFloat("Vector (RGB)"=<0.2903 0.7835 1.5>, "Float (Alpha)"=&RandomRangeFloat(Minimum=0, Maximum=1)))
  ]
  particleUpdateScript [
   &ParticleState()
   &ScaleColor(
    Scale Alpha=&FloatFromCurve(
     FloatCurve=DataInterface() {propertyValues {Curve {keys [<at -0.01 0.0135 RCIM_Constant>,<at 0.362 0.1402 RCIM_Constant>,<at 0.8152 0.3561 RCIM_Constant>,<at 1.0767 0.5048 RCIM_Constant>,<at 1.3382 0.6535 RCIM_Constant>,<at 1.7799 0.9046 RCIM_Constant>,<at 1.9949 1.0269>]}}}
     CurveIndex=$Emitter.Age
     Scale Curve=0.25
    )
    ScaleA=true
    ScaleRGB=true
   )
   &UpdateMeshOrientation(Rotation Rate=0.1, Rotation Vector=$Particles.SpawnRotationVector) {enabled=false}
   &DynamicMaterialParameters(
    Index 0 Param 1=&FloatFromCurve(FloatCurve=DataInterface() {propertyValues {Curve {keys [<at 0.0035 5.7132 RCIM_Cubic>,<at 0.7 8>]}}})
    Index 0 Param 2=&FloatFromCurve(FloatCurve=DataInterface() {propertyValues {Curve {keys [<at -0.1818 1.1125 RCIM_Cubic>,<at 0.9963 2.0443>]}}})
    Index 0 Param 3=1
    Index 0 Param 4=1
    Module.Param0WriteEnabled=true
    Module.Param1WriteEnabled=true
    Module.Param2WriteEnabled=true
    Module.Param3WriteEnabled=true
   )
   &ScaleMeshSize(Scale Factor=&VectorFromFloat(Value=&FloatFromCurve(
      FloatCurve=DataInterface() {propertyValues {Curve {keys [<at -0.0117 0.2808>,<at 0.3596 0.52 RCIM_Constant>,<at 0.6191 0.6872 RCIM_Constant>,<at 0.7784 0.7898 RCIM_Constant>,<at 1.0202 0.9456 RCIM_Constant>,<at 1.4449 1.2192 RCIM_Constant>,<at 1.9935 1.5726>]}}}
      CurveIndex=$Emitter.Age
     )))
   &UpdateMeshOrientation(
    Rotation Rate=&RandomRangeFloat(Minimum=0, Maximum=0.03)
    Rotation Vector=&RandomVector(Vector Scale=1)
   ) {module=UpdateMeshOrientation001}
  ]
  rendererClasses [&NiagaraMeshRendererProperties]
  renderers [{rendererClass=&NiagaraMeshRendererProperties, rendererIndex=0}]
 }
 Vacume_Gas {
  emitterUpdateScript [
   &EmitterState(Life Cycle Mode=::NewEnumerator1, Loop Behavior=::NewEnumerator1, Loop Duration Mode=::NewEnumerator1, UseLoopDelay=false)
   &SpawnBurst_Instantaneous(Spawn Count=1, Use Spawn Probability=false) {enabled=false}
   &SpawnRate(SpawnRate=5, Use Spawn Probability=false)
  ]
  particleSpawnScript [/Niagara/Modules/Spawn/Initialization/V2/InitializeParticle(
    Color Mode=::NewEnumerator0
    Lifetime=&RandomRangeFloat(Minimum=1, Maximum=2)
    Sprite Rotation Mode=::NewEnumerator2
    Sprite Size Mode=::NewEnumerator3
    Uniform Sprite Size=&FloatFromCurve(
     FloatCurve=DataInterface() {propertyValues {Curve {keys [
         {arriveTangent=-0.7198, interpMode=RCIM_Cubic, leaveTangent=-0.7198, tangentMode=RCTM_User, time=0, value=1}
         {arriveTangent=-0.0158, interpMode=RCIM_Cubic, leaveTangent=-0.0158, tangentMode=RCTM_User, time=2.0017, value=0.3244}
        ]}}}
     CurveIndex=$Emitter.Age
     Scale Curve=600
    )
    UsePositionOffset=false
    Write Material Random=true
   )]
  particleUpdateScript [
   &ParticleState(Kill Particles When Lifetime Has Elapsed=false, Let Infinitely Lived Particles Die When Emitter Deactivates=true)
   &Color(
    Color=&MakeLinearColor(
     R:NiagaraFloat=&RandomRangeFloat(Minimum=0.1, Maximum=0.3)
     G:NiagaraFloat=&RandomRangeFloat(Minimum=0.3, Maximum=0.5)
     B:NiagaraFloat=&RandomRangeFloat(Minimum=0.7, Maximum=1)
     A:NiagaraFloat=1
    )
    Scale Alpha=&FloatFromCurve(
     FloatCurve=DataInterface() {propertyValues {Curve {keys [
         <at -0.009 -0.0258 RCIM_Cubic>
         {arriveTangent=0.0256, interpMode=RCIM_Cubic, leaveTangent=0.0256, time=0.4947, value=1}
         <at 1 0 RCIM_Cubic>
        ]}}}
     Scale Curve=&FloatFromCurve(
      FloatCurve=DataInterface() {propertyValues {Curve {keys [<at 0.3486 0.0616 RCIM_Cubic>,<at 2.0065 0.6893 RCIM_Cubic>]}}}
      CurveIndex=$Emitter.Age
      Scale Curve=0.15
     )
    )
   )
   &FN_Radial_Material(
    Base Sprite Offset X:NiagaraFloat=&FloatFromCurve(
     FloatCurve=DataInterface() {propertyValues {Curve {keys [
         {arriveTangent=0.4259, interpMode=RCIM_Cubic, leaveTangent=0.4259, tangentMode=RCTM_User, time=0, value=0}
         {arriveTangent=0.7827, interpMode=RCIM_Cubic, leaveTangent=0.7827, tangentMode=RCTM_User, time=0.9898, value=0.2711}
        ]}}}
     Scale Curve=&FloatFromCurve(
      FloatCurve=DataInterface() {propertyValues {Curve {keys [<at 0 1 RCIM_Cubic>, {arriveTangent=0.2139, interpMode=RCIM_Cubic, leaveTangent=0.2139, tangentMode=RCTM_User, time=1.997, value=1.3893}]}}}
      CurveIndex=$Emitter.Age
     )
    )
    Radial Mask Power=2.5
    Sprite Spin Speed:NiagaraFloat=&RandomRangeFloat(Minimum=-0.05, Maximum=0.05)
   )
   &FN_Radial_Material(
    Base Sprite Offset X:NiagaraFloat=&FloatFromCurve(
     FloatCurve=DataInterface() {propertyValues {Curve {keys [
         {arriveTangent=0.4259, interpMode=RCIM_Cubic, leaveTangent=0.4259, tangentMode=RCTM_User, time=0, value=0}
         {arriveTangent=1.2059, interpMode=RCIM_Cubic, leaveTangent=1.2059, tangentMode=RCTM_User, time=0.986, value=0.9797}
        ]}}}
     Scale Curve=&FloatFromCurve(
      FloatCurve=DataInterface() {propertyValues {Curve {keys [<at 0 1 RCIM_Cubic>, {arriveTangent=1.8317, interpMode=RCIM_Cubic, leaveTangent=1.8317, tangentMode=RCTM_User, time=1.9788, value=2.4147}]}}}
      CurveIndex=$Emitter.Age
     )
    )
    Radial Mask Power=2.5
    Sprite Spin Speed:NiagaraFloat=&RandomRangeFloat(Minimum=-0.05, Maximum=0.05)
   ) {module=FN_Radial_Material001}
   &ScaleSpriteSize(Scale Factor=&Vector2DFromFloat(Value=&FloatFromCurve(FloatCurve=DataInterface() {propertyValues {Curve {keys [<at 0 1 RCIM_Cubic>, {arriveTangent=-1.0068, interpMode=RCIM_Cubic, leaveTangent=-1.0068, tangentMode=RCTM_User, time=0.9995, value=0.3593}]}}})))
   &FN_Sprite_Attributes(
    Color=&MakeLinearColorFromVectorAndFloat(
     "Vector (RGB)"=<1 1 1>
     "Float (Alpha)"=&FloatFromCurve(
      FloatCurve=DataInterface() {propertyValues {Curve {keys [<at -0.0193 0.0327 RCIM_Cubic>, {arriveTangent=-0.032, interpMode=RCIM_Cubic, leaveTangent=-0.032, time=0.501, value=1.0373}, <at 1 0>]}}}
      Scale Curve=&FloatFromCurve(
       FloatCurve=DataInterface() {propertyValues {Curve {keys [<at 0.6888 -0.0087 RCIM_Cubic>,<at 2.0128 1.963 RCIM_Cubic>]}}}
       CurveIndex=$Emitter.Age
      )
     )
    )
    Create Color Attribute=true
    Create SpriteSize Attribute=true
    SpriteSize=&Vector2DFromFloat(Value=&FloatFromCurve(
      FloatCurve=DataInterface() {propertyValues {Curve {keys [<at 0 1>,<at 1 0>]}}}
      Scale Curve=&FloatFromCurve(
       FloatCurve=DataInterface() {propertyValues {Curve {keys [<at -0.0147 0.0065 RCIM_Cubic>,<at 1.9905 0.9951 RCIM_Cubic>]}}}
       CurveIndex=$Emitter.Age
       Scale Curve=200
      )
     ))
    Write to Module Namespace=true
   )
   &SystemLocation(MaskSpawnGroup=false) {enabled=false}
  ]
  rendererClasses [&NiagaraSpriteRendererProperties]
  renderers [
   {rendererClass=&NiagaraSpriteRendererProperties}
   {rendererClass=&NiagaraSpriteRendererProperties, rendererIndex=1}
  ]
 }
 ZeroPointFlare {
  emitterUpdateScript [
   &EmitterState()
   &SpawnBurst_Instantaneous(Spawn Count=1, Use Spawn Probability=false) {enabled=false}
   &SpawnRate(SpawnRate=7, Use Spawn Probability=false)
  ]
  particleSpawnScript [
   /Niagara/Modules/Spawn/Initialization/InitializeParticle(
    Lifetime=&Multiply_Float(
     A=&RandomRangeFloat(Minimum=0.5, Maximum=1.2)
     B=&FloatFromCurve(
      FloatCurve=DataInterface() {propertyValues {Curve {keys [{arriveTangent=-1.5125, interpMode=RCIM_Cubic, leaveTangent=-1.5125, tangentMode=RCTM_User, time=0, value=1}, <at 0.9949 0.2248>]}}}
      CurveIndex=$Emitter.NormalizedLoopAge
     )
    )
    Module.Write Color=true
    Module.Write Lifetime=true
    Module.Write Mass=true
    Module.Write Material Random=true
    Module.Write Position=false
    Module.Write Scale=false
    Module.Write SpriteRotation=true
    Module.Write SpriteSize=true
    Sprite Rotation=&RandomRangeFloat(Minimum=0, Maximum=360)
    Sprite Size=<3000 3000>
   )
   &Color(Color=&MakeLinearColorFromVectorAndFloat("Vector (RGB)"=<0.0175 0.0243 0.03>, "Float (Alpha)"=&ClampFloat(Float=0.6)))
  ]
  particleUpdateScript [
   &ParticleState()
   &ScaleColor(
    Scale Alpha=&FloatFromCurve(FloatCurve=DataInterface() {propertyValues {Curve {keys [<at 0 1 RCIM_Cubic>,<at 0.25 1 RCIM_Cubic>,<at 0.75 1 RCIM_Cubic>,<at 1 1>]}}})
    ScaleA=true
    ScaleRGB=true
   )
   /SetVariables(Particles.Size:NiagaraFloat=&FloatFromCurve(
     FloatCurve=DataInterface() {propertyValues {Curve {keys [
         <at -0.0039 0.0032 RCIM_Cubic>
         {arriveTangent=0.0358, interpMode=RCIM_Cubic, leaveTangent=0.0358, tangentMode=RCTM_User, time=0.4967, value=0.9107}
         <at 1 0>
        ]}}}
     Scale Curve=&FloatFromCurve(
      FloatCurve=DataInterface() {propertyValues {Curve {keys [
          <at -0.0005 0.984 RCIM_Cubic>
          {arriveTangent=8.0076, interpMode=RCIM_Cubic, leaveTangent=8.0076, tangentMode=RCTM_User, time=1.0095, value=3.8786}
         ]}}}
      CurveIndex=$Emitter.NormalizedLoopAge
      Scale Curve=0.5
     )
    )) {isSetParameters=true}
   &FN_Sprite_Attributes(
    Camera Offset=-250
    Create Camera Offset Attribute=true
    Create Color Attribute=true
    Create SpriteSize Attribute=true
    SpriteSize=&Vector2DFromFloat(Value=&Multiply_Float(A=500, B=$Particles.Size))
    Write to Module Namespace=true
   )
   &FN_Sprite_Attributes(
    Camera Offset=50
    Color=&MakeLinearColorFromVectorAndFloat("Vector (RGB)"=<0.5625 1.8964 3>, "Float (Alpha)"=0.3)
    Create Camera Offset Attribute=true
    Create Color Attribute=true
    Create SpriteSize Attribute=true
    SpriteSize=&Vector2DFromFloat(Value=&Multiply_Float(A=400, B=$Particles.Size))
    Write to Module Namespace=true
   ) {module=FN_Sprite_Attributes001}
   &FN_Sprite_Attributes(
    Camera Offset=50
    Color=&MakeLinearColorFromVectorAndFloat(
     "Vector (RGB)"=<0.8602 1.4152 2>
     "Float (Alpha)"=&FloatFromCurve(
      FloatCurve=DataInterface() {propertyValues {Curve {keys [
          <at 0.0048 -0.0024 RCIM_Cubic>
          {arriveTangent=2.9233, interpMode=RCIM_Cubic, leaveTangent=2.9233, tangentMode=RCTM_User, time=0.7236, value=0.9786}
          <at 0.9988 -0.0024>
         ]}}}
      Scale Curve=&FloatFromCurve(
       FloatCurve=DataInterface() {propertyValues {Curve {keys [<at 0.5027 0>,<at 0.9051 1>]}}}
       CurveIndex=$Emitter.NormalizedLoopAge
       Scale Curve=0.3
      )
     )
    )
    Create Camera Offset Attribute=true
    Create Color Attribute=true
    Create SpriteSize Attribute=true
    SpriteSize=&Vector2DFromFloat(Value=&Multiply_Float(
      A=400
      B=&FloatFromCurve(
       FloatCurve=DataInterface() {propertyValues {Curve {keys [
           <at 0.004 1.0518 RCIM_Cubic>
           {arriveTangent=-0.8729, interpMode=RCIM_Cubic, leaveTangent=-0.8729, tangentMode=RCTM_User, time=0.9881, value=0.7637}
          ]}}}
       Scale Curve=2
      )
     ))
    Write to Module Namespace=true
   ) {module=FN_Sprite_Attributes002}
   &FN_Sprite_Attributes(
    Color=&MakeLinearColorFromVectorAndFloat(
     "Vector (RGB)"=<1.5665 2.3369 6>
     "Float (Alpha)"=&FloatFromCurve(
      FloatCurve=DataInterface() {propertyValues {Curve {keys [
          <at 0.0048 -0.0024 RCIM_Cubic>
          {arriveTangent=2.9233, interpMode=RCIM_Cubic, leaveTangent=2.9233, tangentMode=RCTM_User, time=0.7236, value=0.9786}
          <at 0.9988 -0.0024>
         ]}}}
      Scale Curve=0.5
     )
    )
    Create Color Attribute=true
    Create Sprite Rotation Attribute=true
    Create SpriteSize Attribute=true
    SpriteSize=&Multiply_Vector2(
     A=<1000 20>
     B=&Vector2DFromFloat(Value=&RandomRangeFloat(Minimum=0.2, Maximum=1))
    )
    Write to Module Namespace=true
   ) {module=FN_Sprite_Attributes003}
   &FN_Sprite_Attributes(
    Color=&MakeLinearColorFromVectorAndFloat("Vector (RGB)"=<0.2398 1.0892 1.5>, "Float (Alpha)"=&ClampFloat(Float=0.3))
    Create Color Attribute=true
    Create Sprite Rotation Attribute=true
    Create SpriteSize Attribute=true
    SpriteSize=&Multiply_Vector2(
     A=&RandomRangeVector2D(Minimum=<120 80>, Maximum=<200 140>, Evaluation Type=::NewEnumerator1)
     B=&Vector2DFromFloat(Value=&FloatFromCurve(
       FloatCurve=DataInterface() {propertyValues {Curve {keys [
           <at 0 0 RCIM_Cubic>
           {arriveTangent=1.4286, interpMode=RCIM_Cubic, leaveTangent=1.4286, time=0.5754, value=0.3493}
           {arriveTangent=-0.8228, interpMode=RCIM_Cubic, leaveTangent=-0.8228, time=0.7, value=1}
           <at 1 0>
          ]}}}
       Scale Curve=5
      ))
    )
    Write to Module Namespace=true
   ) {module=FN_Sprite_Attributes004}
   &FN_Light_Attributes(
    Color=<rgba 1 2 10 1>
    Create Color Attribute=true
    Create Position Attribute=true
    Enable Visibility Tag=true
    Position=&Add_Vector(
     A=Linked() {linkedVariable {name=Particles.Position, type=NiagaraPosition}}
     B=&RandomVector(Vector Scale=20, Evaluation Type=::NewEnumerator1)
    )
    Scale Color=&FloatFromCurve(
     FloatCurve=DataInterface() {propertyValues {Curve {keys [<at -0.0008 -0.006 RCIM_Cubic>, {arriveTangent=0.006, interpMode=RCIM_Cubic, leaveTangent=0.006, time=0.4887, value=0.9748}, <at 1 0>]}}}
     Scale Curve=10000
    )
    VisibilityTag=HlslExpression() {hlslExpression="Particles.UniqueID && 1"}
   )
   &FN_Gen5_Light(Color:LinearColor=<rgba 1 2 10 1>, Intensity=10000)
  ]
  rendererClasses [
   &NiagaraSpriteRendererProperties
   &NiagaraComponentRendererProperties
   &NiagaraLightRendererProperties
  ]
  renderers [
   {rendererClass=&NiagaraSpriteRendererProperties}
   {rendererClass=&NiagaraSpriteRendererProperties, rendererIndex=1}
   {rendererClass=&NiagaraSpriteRendererProperties, rendererIndex=2}
   {rendererClass=&NiagaraSpriteRendererProperties, rendererIndex=3}
   {rendererClass=&NiagaraSpriteRendererProperties, rendererIndex=4}
   {rendererClass=&NiagaraComponentRendererProperties, rendererIndex=5}
   {rendererClass=&NiagaraLightRendererProperties, rendererIndex=6}
  ]
 }
}
summary {
 systemName=NS_JFly_Teleport_Charge
 emitters [
  {
   emitterName=ZeroPointFlare
   rendererClasses [
    &NiagaraSpriteRendererProperties
    &NiagaraComponentRendererProperties
    &NiagaraLightRendererProperties
   ]
  }
  {emitterName=Embers, simTarget=GPUComputeSim, rendererClasses [&NiagaraSpriteRendererProperties]}
  {emitterName=LightbeamsWarmup, rendererClasses [&NiagaraMeshRendererProperties]}
  {emitterName=Crack, rendererClasses [&NiagaraMeshRendererProperties]}
  {emitterName=Vacume_Gas, rendererClasses [&NiagaraSpriteRendererProperties]}
  {emitterName=ElectricArcsResidual, rendererClasses [&NiagaraMeshRendererProperties]}
 ]
 userVariables []
}
